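# Unlit additive material for a sharp cross flare in the LimeRock countdown timer's treehouse pre-event effects, usable on particle sprites.
# Particle or parameter color tints a cross-flare texture, with a hot center added outside low quality and mobile.
# Depth, near-camera and optional view-angle fades attenuate it.
# With the default-on Sprite switch off, a Sprite function drives world position offset, sized by time-offset sine pulses masked by vertex color RGB and camera distance.
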Material /LimeRock_CountdownTimer/Assets/FX/Treehouse/PreEvent_Stage02/M_LimeRock_Sharp_Cross_Flare
graph {
 properties {
  bUsedWithNiagaraSprites=true
  bUsedWithParticleSprites=true
  blendMode=BLEND_Additive
  pixelDepthOffsetMode=PDOM_Legacy
  shadingModel=MSM_Unlit
  usedShadingModels=Unlit
 }
 nodes [
  n1 = ParticleColor()
  n2 = TextureSampleParameter2D() {parameterName=Texture, samplerType=SAMPLERTYPE_Grayscale, texture=/Game/Effects/Fort_Effects/Textures/Flares/T_Cross_Flare}
  n3 = ScalarParameter() {defaultValue=150, parameterName=NearFadeDistance}
  n4 = Constant() {r=1}
  n5 = QualitySwitch(
   Default=DepthFade(
    Opacity=MaterialFunctionCall(Distance=@n3) {materialFunction=&Near_Camera_Fade}.Result
    FadeDistance=ScalarParameter() {defaultValue=30, parameterName=DepthFadeDistance}
   )
   Low=@n4
  )
  n6 = Multiply(A=Multiply(A=@n1.A, B=@n2.R), B=FeatureLevelSwitch(Default=@n5, ES2=@n4, ES3_1=@n4, SM4_REMOVED=@n5))
  n7 = CameraVectorWS()
  n8 = Transform(None=Constant3Vector() {constant=<rgba 1 0 0 0>}) {transformSourceType=TRANSFORMSOURCE_Local}
  LinearInterpolate(A=@n6, B=Multiply(A=@n6, B=DotProduct(A=@n7, B=@n8)), Alpha=ScalarParameter() {parameterName=UseViewAngleFade})
  n9 = StaticSwitchParameter(True=@n1.RGB, False=VectorParameter() {parameterName=Base Color, defaultValue=<rgba 0.6854 1.6198 2.8 1>}.RGB) {defaultValue=true, parameterName=Base Color}
  n10 = LinearInterpolate(A=@n9) {constAlpha=0.2}
  n11 = Power(Base=SphereMask(A=TextureCoordinate(), B=Constant() {r=0.5}) {attenuationRadius=0.08, hardnessPercent=0}) {constExponent=8}
  n12 = Constant()
  n13 = QualitySwitch(Default=Multiply(A=Multiply(A=@n10, B=@n11)) {constB=2}, Low=@n12)
  n14 = LinearInterpolate(A=VectorParameter() {parameterName=DarkColor, defaultValue=<rgba 0.1133 0.088 0.2448 1>}.RGB, Alpha=@n2.R)
  MaterialFunctionCall(Emissive=Add(A=ShadingPathSwitch(Default=@n13, Mobile=@n12), B=Multiply(A=@n9, B=@n14))) {materialFunction=&MF_FX_EmissiveAdjustments}
  ScalarParameter() {defaultValue=50, parameterName=SpriteScaleFar}
  n15 = VertexColor()
  n16 = Time()
  n17 = Add(A=Sine(None=Multiply(A=@n16) {constB=0.32}))
  n18 = Multiply(A=@n15.R, B=Power(Base=Multiply(A=@n17) {constB=0.5}) {constExponent=1})
  n19 = Add(A=Time()) {constB=4.5}
  n20 = Add(A=Sine(None=Multiply(A=@n19) {constB=0.32}))
  n21 = Multiply(A=@n15.G, B=Power(Base=Multiply(A=@n20) {constB=0.5}))
  n22 = Time()
  n23 = Sine(None=Multiply(A=Add(A=@n22) {constB=11.5}) {constB=0.4})
  n24 = Power(Base=Multiply(A=Add(A=@n23)) {constB=0.5}) {constExponent=1.5}
  n25 = Multiply(A=Add(A=Add(A=@n18, B=@n21), B=Multiply(A=@n15.B, B=@n24)), B=ScalarParameter() {defaultValue=120, parameterName=FlareSize})
  n26 = ScalarParameter() {defaultValue=580, parameterName=SoftenDistance}
  n27 = MaterialFunctionCall(Distance=ScalarParameter() {defaultValue=1250, parameterName=Camera Size}) {materialFunction=&Near_Camera_Fade}
  n28 = Multiply(
   A=LinearInterpolate(
    A=@n25
    B=@n25
    Alpha=MaterialFunctionCall(Distance=@n26) {materialFunction=&Near_Camera_Fade}.Result
   )
   B=StaticSwitchParameter(True=@n27.Result, False=OneMinus(None=@n27.Result)) {parameterName=ScaleDownSizeWhenCloseToCam}
  )
  StaticSwitchParameter(
   True=StaticSwitchParameter(True=MaterialFunctionCall() {materialFunction=/Game/Environments/Sets/DigSite/Materials/MF_Sneezy_DanceClub}.Result, False=Constant()) {parameterName=DanceClub}
   False=MaterialFunctionCall!(Center Location=WorldPosition().XYZ, 0-1 UVs, XY scale=AppendVector(A=@n28, B=@n28), OverridePivot, "Normals (Optional)") {materialFunction=&Sprite}.World position offset
  ) {defaultValue=true, parameterName=Sprite}
 ]
}
